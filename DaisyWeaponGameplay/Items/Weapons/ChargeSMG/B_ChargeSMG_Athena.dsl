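; Weapon Blueprint for the Charge SMG in the Athena game mode, adding helpers that track ammo and charge.
; AmmoRemaining caches the local magazine ammo count from FortWeapon into a variable.
; ChargeAmount raises the stored charge by a fixed step on each call, clamped between zero and a maximum.
; The construction script only calls the parent, and no callers of either helper appear here.
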(fn AmmoRemaining
  (Variables|Default|SetAmmoRemaining (FortWeapon|GetLocalMagazineAmmoCount)))

(fn ChargeAmount
  (bind _output_get (Variables|Default|SetChargeAmount (Math|Integer|Clamp(Integer) (+ (Variables|Default|GetChargeAmount) 3) 0 30))))

(fn ConstructionScript
  (|CallParentFunction))
