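# Animation Blueprint for the Charge SMG weapon mesh, animating its charge dial and visible ammunition.
# Each update reads charge, firing, ammo and reload state from B_ChargeSMG_Athena.
# The dial bone turns with charge, then eases back to zero over time proportional to the last charge.
# Per-bone blending layers two state machines onto bullet bones: magazine rotation in three-round steps with reload handling, and bullet insertion with last-shot and empty states.
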AnimBlueprint /DaisyWeaponGameplay/Items/Weapons/ChargeSMG/ChargeSMG_AnimBP
target_skeleton=/DaisyWeaponGameplay/Items/Weapons/ChargeSMG/Meshes/Charge_SMG_Skeleton
graphs [
 AnimGraph {nodes [
   Copy Pose From Mesh() {bCopyCurves=true}
   n1 = Input Pose()
   n2 = "Transform (Modify) Bone - Bone: dial"!(
    ComponentPose=Local To Component(LocalPose=Slot 'DefaultSlot'(Source=@n1))
    Alpha=1
    Rotation=MakeRotator(Roll=Get ChargeRotation() -> (ChargeRotation:double))
   ) -> (Pose) {rotationMode=BMM_Replace, rotationSpace=BCS_BoneSpace, alphaBoolBlend {blendOption=Linear}, boneToModify {boneName=dial}}
   AnimGraphNode_StateMachine_0 = MagBulletRotation_StateMachine() {is=title AnimGraphNode_StateMachine}
   AnimGraphNode_StateMachine_1 = BulletInsert_StateMachine() {is=title AnimGraphNode_StateMachine}
   n3 = Layered blend per bone!(
    BasePose=Component To Local(ComponentPose=@n2)
    BlendPoses_0=@AnimGraphNode_StateMachine_0
    BlendPoses_1:Pose Link Structure=@AnimGraphNode_StateMachine_1
    BlendWeights_0
    BlendWeights_1:float=1
   ) -> (Pose) {
    blendPoses [{linkId=-1, sourceLinkId=-1}, {linkId=-1, sourceLinkId=-1}]
    blendWeights [1,1]
    layerSetup [{branchFilters [{boneName=bullets}]}, {branchFilters [{boneName=bullets_gun_01}, {boneName=bullets_gun_02}]}]
   }
   Output Pose(Result=@n3) {layerGroup=None}
  ]}
 MagBulletRotation_StateMachine {
  entry=StartPose
  path=AnimGraph.AnimGraphNode_StateMachine_0.MagBulletRotation_StateMachine
  ids=id
  nodes [
   AnimStateNode_0 = 24 pose() {
    is=title AnimStateNode
    stateEntered {guid=#1}
    stateFullyBlended {guid=#2}
    stateLeft {guid=#3}
   }
   AnimStateNode_1 = 21 pose() {
    is=title AnimStateNode
    stateEntered {guid=#4}
    stateFullyBlended {guid=#5}
    stateLeft {guid=#6}
   }
   AnimStateNode_13 = Full pose() {is=title AnimStateNode}
   AnimStateNode_14 = 27 pose() {
    is=title AnimStateNode
    stateEntered {guid=#7}
    stateFullyBlended {guid=#8}
    stateLeft {guid=#9}
   }
   AnimStateNode_2 = 18 pose() {
    is=title AnimStateNode
    stateEntered {guid=#10}
    stateFullyBlended {guid=#11}
    stateLeft {guid=#12}
   }
   AnimStateNode_3 = 15 pose() {
    is=title AnimStateNode
    stateEntered {guid=#13}
    stateFullyBlended {guid=#14}
    stateLeft {guid=#15}
   }
   AnimStateNode_4 = 12 pose() {
    is=title AnimStateNode
    stateEntered {guid=#16}
    stateFullyBlended {guid=#17}
    stateLeft {guid=#18}
   }
   AnimStateNode_5 = 9 pose() {
    is=title AnimStateNode
    stateEntered {guid=#19}
    stateFullyBlended {guid=#20}
    stateLeft {guid=#21}
   }
   AnimStateNode_56 = IsReloading() {is=title AnimStateNode}
   AnimStateNode_57 = StartPose() {
    is=title AnimStateNode
    stateEntered {guid=#22}
    stateFullyBlended {guid=#23}
    stateLeft {guid=#24}
   }
   AnimStateNode_6 = 6 pose() {
    is=title AnimStateNode
    stateEntered {guid=#25}
    stateFullyBlended {guid=#26}
    stateLeft {guid=#27}
   }
   AnimStateNode_7 = 3 pose() {
    is=title AnimStateNode
    stateEntered {guid=#28}
    stateFullyBlended {guid=#29}
    stateLeft {guid=#30}
   }
   AnimStateNode_9 = Empty pose() {is=title AnimStateNode}
   AnimStateTransitionNode_100 = AnimStateTransitionNode()
   AnimStateTransitionNode_101 = AnimStateTransitionNode()
   AnimStateTransitionNode_12 = AnimStateTransitionNode()
   AnimStateTransitionNode_13 = AnimStateTransitionNode()
   AnimStateTransitionNode_14 = AnimStateTransitionNode() {crossfadeDuration=0}
   AnimStateTransitionNode_24 = AnimStateTransitionNode()
   AnimStateTransitionNode_26 = AnimStateTransitionNode()
   AnimStateTransitionNode_27 = AnimStateTransitionNode()
   AnimStateTransitionNode_29 = AnimStateTransitionNode()
   AnimStateTransitionNode_30 = AnimStateTransitionNode()
   AnimStateTransitionNode_31 = AnimStateTransitionNode()
   AnimStateTransitionNode_32 = AnimStateTransitionNode()
   AnimStateTransitionNode_33 = AnimStateTransitionNode()
   AnimStateTransitionNode_34 = AnimStateTransitionNode()
   AnimStateTransitionNode_35 = AnimStateTransitionNode()
   AnimStateTransitionNode_36 = AnimStateTransitionNode() {
    transitionEnd {guid=#31}
    transitionInterrupt {guid=#32}
    transitionStart {guid=#33}
   }
   AnimStateTransitionNode_37 = AnimStateTransitionNode() {
    transitionEnd {guid=#34}
    transitionInterrupt {guid=#35}
    transitionStart {guid=#36}
   }
   AnimStateTransitionNode_38 = AnimStateTransitionNode() {
    transitionEnd {guid=#37}
    transitionInterrupt {guid=#38}
    transitionStart {guid=#39}
   }
   AnimStateTransitionNode_39 = AnimStateTransitionNode() {
    transitionEnd {guid=#40}
    transitionInterrupt {guid=#41}
    transitionStart {guid=#42}
   }
   AnimStateTransitionNode_40 = AnimStateTransitionNode() {
    transitionEnd {guid=#43}
    transitionInterrupt {guid=#44}
    transitionStart {guid=#45}
   }
   AnimStateTransitionNode_41 = AnimStateTransitionNode() {
    transitionEnd {guid=#46}
    transitionInterrupt {guid=#47}
    transitionStart {guid=#48}
   }
   AnimStateTransitionNode_42 = AnimStateTransitionNode() {
    transitionEnd {guid=#49}
    transitionInterrupt {guid=#50}
    transitionStart {guid=#51}
   }
   AnimStateTransitionNode_43 = AnimStateTransitionNode() {
    transitionEnd {guid=#52}
    transitionInterrupt {guid=#53}
    transitionStart {guid=#54}
   }
   AnimStateTransitionNode_44 = AnimStateTransitionNode() {
    transitionEnd {guid=#55}
    transitionInterrupt {guid=#56}
    transitionStart {guid=#57}
   }
   AnimStateTransitionNode_6 = AnimStateTransitionNode()
   AnimStateTransitionNode_7 = AnimStateTransitionNode() {priorityOrder=2}
   AnimStateTransitionNode_9 = AnimStateTransitionNode()
   AnimStateTransitionNode_96 = AnimStateTransitionNode()
   AnimStateTransitionNode_97 = AnimStateTransitionNode() {
    priorityOrder=2
    transitionEnd {guid=#58}
    transitionInterrupt {guid=#59}
    transitionStart {guid=#60}
   }
   AnimStateTransitionNode_98 = AnimStateTransitionNode() {priorityOrder=2}
   AnimStateTransitionNode_99 = AnimStateTransitionNode()
  ]
  links [
   {
    dst=9 pose
    rule=AnimGraph.AnimGraphNode_StateMachine_0.MagBulletRotation_StateMachine.AnimStateTransitionNode_7.Transition
    src=12 pose
    via=AnimStateTransitionNode_7
   }
   {
    dst=IsReloading
    rule=AnimGraph.AnimGraphNode_StateMachine_0.MagBulletRotation_StateMachine.AnimStateTransitionNode_27.Transition
    src=12 pose
    via=AnimStateTransitionNode_27
   }
   {
    dst=12 pose
    rule=AnimGraph.AnimGraphNode_StateMachine_0.MagBulletRotation_StateMachine.AnimStateTransitionNode_6.Transition
    src=15 pose
    via=AnimStateTransitionNode_6
   }
   {
    dst=IsReloading
    rule=AnimGraph.AnimGraphNode_StateMachine_0.MagBulletRotation_StateMachine.AnimStateTransitionNode_32.Transition
    src=15 pose
    via=AnimStateTransitionNode_32
   }
   {
    dst=15 pose
    rule=AnimGraph.AnimGraphNode_StateMachine_0.MagBulletRotation_StateMachine.AnimStateTransitionNode_100.Transition
    src=18 pose
    via=AnimStateTransitionNode_100
   }
   {
    dst=IsReloading
    rule=AnimGraph.AnimGraphNode_StateMachine_0.MagBulletRotation_StateMachine.AnimStateTransitionNode_31.Transition
    src=18 pose
    via=AnimStateTransitionNode_31
   }
   {
    dst=18 pose
    rule=AnimGraph.AnimGraphNode_StateMachine_0.MagBulletRotation_StateMachine.AnimStateTransitionNode_99.Transition
    src=21 pose
    via=AnimStateTransitionNode_99
   }
   {
    dst=IsReloading
    rule=AnimGraph.AnimGraphNode_StateMachine_0.MagBulletRotation_StateMachine.AnimStateTransitionNode_30.Transition
    src=21 pose
    via=AnimStateTransitionNode_30
   }
   {
    dst=21 pose
    rule=AnimGraph.AnimGraphNode_StateMachine_0.MagBulletRotation_StateMachine.AnimStateTransitionNode_98.Transition
    src=24 pose
    via=AnimStateTransitionNode_98
   }
   {
    dst=IsReloading
    rule=AnimGraph.AnimGraphNode_StateMachine_0.MagBulletRotation_StateMachine.AnimStateTransitionNode_26.Transition
    src=24 pose
    via=AnimStateTransitionNode_26
   }
   {
    dst=24 pose
    rule=AnimGraph.AnimGraphNode_StateMachine_0.MagBulletRotation_StateMachine.AnimStateTransitionNode_97.Transition
    src=27 pose
    via=AnimStateTransitionNode_97
   }
   {
    dst=IsReloading
    rule=AnimGraph.AnimGraphNode_StateMachine_0.MagBulletRotation_StateMachine.AnimStateTransitionNode_24.Transition
    src=27 pose
    via=AnimStateTransitionNode_24
   }
   {
    dst=Empty pose
    rule=AnimGraph.AnimGraphNode_StateMachine_0.MagBulletRotation_StateMachine.AnimStateTransitionNode_12.Transition
    src=3 pose
    via=AnimStateTransitionNode_12
   }
   {
    dst=IsReloading
    rule=AnimGraph.AnimGraphNode_StateMachine_0.MagBulletRotation_StateMachine.AnimStateTransitionNode_35.Transition
    src=3 pose
    via=AnimStateTransitionNode_35
   }
   {
    dst=3 pose
    rule=AnimGraph.AnimGraphNode_StateMachine_0.MagBulletRotation_StateMachine.AnimStateTransitionNode_9.Transition
    src=6 pose
    via=AnimStateTransitionNode_9
   }
   {
    dst=IsReloading
    rule=AnimGraph.AnimGraphNode_StateMachine_0.MagBulletRotation_StateMachine.AnimStateTransitionNode_34.Transition
    src=6 pose
    via=AnimStateTransitionNode_34
   }
   {
    dst=6 pose
    rule=AnimGraph.AnimGraphNode_StateMachine_0.MagBulletRotation_StateMachine.AnimStateTransitionNode_101.Transition
    src=9 pose
    via=AnimStateTransitionNode_101
   }
   {
    dst=IsReloading
    rule=AnimGraph.AnimGraphNode_StateMachine_0.MagBulletRotation_StateMachine.AnimStateTransitionNode_33.Transition
    src=9 pose
    via=AnimStateTransitionNode_33
   }
   {
    dst=IsReloading
    rule=AnimGraph.AnimGraphNode_StateMachine_0.MagBulletRotation_StateMachine.AnimStateTransitionNode_13.Transition
    src=Empty pose
    via=AnimStateTransitionNode_13
   }
   {
    dst=27 pose
    rule=AnimGraph.AnimGraphNode_StateMachine_0.MagBulletRotation_StateMachine.AnimStateTransitionNode_96.Transition
    src=Full pose
    via=AnimStateTransitionNode_96
   }
   {
    dst=Full pose
    rule=AnimGraph.AnimGraphNode_StateMachine_0.MagBulletRotation_StateMachine.AnimStateTransitionNode_14.Transition
    src=IsReloading
    via=AnimStateTransitionNode_14
   }
   {
    dst=StartPose
    rule=AnimGraph.AnimGraphNode_StateMachine_0.MagBulletRotation_StateMachine.AnimStateTransitionNode_29.Transition
    src=IsReloading
    via=AnimStateTransitionNode_29
   }
   {
    dst=12 pose
    rule=AnimGraph.AnimGraphNode_StateMachine_0.MagBulletRotation_StateMachine.AnimStateTransitionNode_41.Transition
    src=StartPose
    via=AnimStateTransitionNode_41
   }
   {
    dst=15 pose
    rule=AnimGraph.AnimGraphNode_StateMachine_0.MagBulletRotation_StateMachine.AnimStateTransitionNode_40.Transition
    src=StartPose
    via=AnimStateTransitionNode_40
   }
   {
    dst=18 pose
    rule=AnimGraph.AnimGraphNode_StateMachine_0.MagBulletRotation_StateMachine.AnimStateTransitionNode_39.Transition
    src=StartPose
    via=AnimStateTransitionNode_39
   }
   {
    dst=21 pose
    rule=AnimGraph.AnimGraphNode_StateMachine_0.MagBulletRotation_StateMachine.AnimStateTransitionNode_38.Transition
    src=StartPose
    via=AnimStateTransitionNode_38
   }
   {
    dst=24 pose
    rule=AnimGraph.AnimGraphNode_StateMachine_0.MagBulletRotation_StateMachine.AnimStateTransitionNode_37.Transition
    src=StartPose
    via=AnimStateTransitionNode_37
   }
   {
    dst=27 pose
    rule=AnimGraph.AnimGraphNode_StateMachine_0.MagBulletRotation_StateMachine.AnimStateTransitionNode_36.Transition
    src=StartPose
    via=AnimStateTransitionNode_36
   }
   {
    dst=3 pose
    rule=AnimGraph.AnimGraphNode_StateMachine_0.MagBulletRotation_StateMachine.AnimStateTransitionNode_44.Transition
    src=StartPose
    via=AnimStateTransitionNode_44
   }
   {
    dst=6 pose
    rule=AnimGraph.AnimGraphNode_StateMachine_0.MagBulletRotation_StateMachine.AnimStateTransitionNode_43.Transition
    src=StartPose
    via=AnimStateTransitionNode_43
   }
   {
    dst=9 pose
    rule=AnimGraph.AnimGraphNode_StateMachine_0.MagBulletRotation_StateMachine.AnimStateTransitionNode_42.Transition
    src=StartPose
    via=AnimStateTransitionNode_42
   }
  ]
 }
 24 pose {
  path=AnimGraph.AnimGraphNode_StateMachine_0.MagBulletRotation_StateMachine.AnimStateNode_0.24 pose
  nodes [
   n1 = Sequence Player 'ChargeSMG_Mag_24to21_Charged_Rifle'() {bLoopAnimation=false, sequence=/DaisyWeaponGameplay/Items/Weapons/ChargeSMG/Animation/MagazineRotation/ChargeSMG_Mag_24to21_Charged_Rifle}
   Output Animation Pose(Result=@n1)
  ]
 }
 21 pose {
  path=AnimGraph.AnimGraphNode_StateMachine_0.MagBulletRotation_StateMachine.AnimStateNode_1.21 pose
  nodes [
   n1 = Sequence Player 'ChargeSMG_Mag_21to18_Charged_Rifle'() {bLoopAnimation=false, sequence=/DaisyWeaponGameplay/Items/Weapons/ChargeSMG/Animation/MagazineRotation/ChargeSMG_Mag_21to18_Charged_Rifle}
   Output Animation Pose(Result=@n1)
  ]
 }
 Full pose {
  path=AnimGraph.AnimGraphNode_StateMachine_0.MagBulletRotation_StateMachine.AnimStateNode_13.Full pose
  nodes [Output Animation Pose(Result=Sequence Player 'ChargeSMG_BulletFull_Pose_Charged_Rifle'() {sequence=/DaisyWeaponGameplay/Items/Weapons/ChargeSMG/Animation/MagazineRotation/ChargeSMG_BulletFull_Pose_Charged_Rifle})]
 }
 27 pose {
  path=AnimGraph.AnimGraphNode_StateMachine_0.MagBulletRotation_StateMachine.AnimStateNode_14.27 pose
  nodes [
   n1 = Sequence Player 'ChargeSMG_Mag_27to24_Charged_Rifle'() {bLoopAnimation=false, sequence=/DaisyWeaponGameplay/Items/Weapons/ChargeSMG/Animation/MagazineRotation/ChargeSMG_Mag_27to24_Charged_Rifle}
   Output Animation Pose(Result=@n1)
  ]
 }
 18 pose {
  path=AnimGraph.AnimGraphNode_StateMachine_0.MagBulletRotation_StateMachine.AnimStateNode_2.18 pose
  nodes [
   n1 = Sequence Player 'ChargeSMG_Mag_18to15_Charged_Rifle'() {bLoopAnimation=false, sequence=/DaisyWeaponGameplay/Items/Weapons/ChargeSMG/Animation/MagazineRotation/ChargeSMG_Mag_18to15_Charged_Rifle}
   Output Animation Pose(Result=@n1)
  ]
 }
 15 pose {
  path=AnimGraph.AnimGraphNode_StateMachine_0.MagBulletRotation_StateMachine.AnimStateNode_3.15 pose
  nodes [
   n1 = Sequence Player 'ChargeSMG_Mag_15to12_Charged_Rifle'() {bLoopAnimation=false, sequence=/DaisyWeaponGameplay/Items/Weapons/ChargeSMG/Animation/MagazineRotation/ChargeSMG_Mag_15to12_Charged_Rifle}
   Output Animation Pose(Result=@n1)
  ]
 }
 12 pose {
  path=AnimGraph.AnimGraphNode_StateMachine_0.MagBulletRotation_StateMachine.AnimStateNode_4.12 pose
  nodes [Output Animation Pose(Result=Sequence Player 'ChargeSMG_Mag_12to09_Charged_Rifle'() {sequence=/DaisyWeaponGameplay/Items/Weapons/ChargeSMG/Animation/MagazineRotation/ChargeSMG_Mag_12to09_Charged_Rifle})]
 }
 9 pose {
  path=AnimGraph.AnimGraphNode_StateMachine_0.MagBulletRotation_StateMachine.AnimStateNode_5.9 pose
  nodes [Output Animation Pose(Result=Sequence Player 'ChargeSMG_Mag_09to06_Charged_Rifle'() {sequence=/DaisyWeaponGameplay/Items/Weapons/ChargeSMG/Animation/MagazineRotation/ChargeSMG_Mag_09to06_Charged_Rifle})]
 }
 IsReloading {
  path=AnimGraph.AnimGraphNode_StateMachine_0.MagBulletRotation_StateMachine.AnimStateNode_56.IsReloading
  nodes [
   n1 = Get BulletsUsed() -> (BulletsUsed:int=0)
   "Sequence Evaluator (by time) 'ChargeSMG_BulletRotation_Charged_Rifle'"!(ExplicitTime:float=float / float(A:int=int / int(A=@n1, B=3), B:int=10)) -> (Pose:Pose Link Structure) {
    is=title AnimGraphNode_SequenceEvaluator
    bTeleportToExplicitTime=false
    reinitializationBehavior=NoReset
    sequence=/DaisyWeaponGameplay/Items/Weapons/ChargeSMG/Animation/MagazineRotation/ChargeSMG_BulletRotation_Charged_Rifle
   }
   Output Animation Pose(Result=Sequence Player 'ChargeSMG_Reload_Weapon_CMM_Weapon'() {sequence=/DaisyWeaponGameplay/Items/Weapons/ChargeSMG/Animation/ChargeSMG_Reload_Weapon_CMM_Weapon})
  ]
 }
 StartPose {
  path=AnimGraph.AnimGraphNode_StateMachine_0.MagBulletRotation_StateMachine.AnimStateNode_57.StartPose
  nodes [
   n1 = Get BulletsUsed() -> (BulletsUsed=0)
   n2 = "Sequence Evaluator (by time) 'ChargeSMG_BulletRotation_Charged_Rifle'"(ExplicitTime=float / float(A=int / int(A=@n1, B=3), B=10)) {
    bTeleportToExplicitTime=false
    reinitializationBehavior=NoReset
    sequence=&ChargeSMG_BulletRotation_Charged_Rifle
   }
   Output Animation Pose(Result=@n2)
  ]
 }
 6 pose {
  path=AnimGraph.AnimGraphNode_StateMachine_0.MagBulletRotation_StateMachine.AnimStateNode_6.6 pose
  nodes [Output Animation Pose(Result=Sequence Player 'ChargeSMG_Mag_06to03_Charged_Rifle'() {sequence=/DaisyWeaponGameplay/Items/Weapons/ChargeSMG/Animation/MagazineRotation/ChargeSMG_Mag_06to03_Charged_Rifle})]
 }
 3 pose {
  path=AnimGraph.AnimGraphNode_StateMachine_0.MagBulletRotation_StateMachine.AnimStateNode_7.3 pose
  nodes [Output Animation Pose(Result=Sequence Player 'ChargeSMG_Mag_03to00_Charged_Rifle'() {sequence=/DaisyWeaponGameplay/Items/Weapons/ChargeSMG/Animation/MagazineRotation/ChargeSMG_Mag_03to00_Charged_Rifle})]
 }
 Empty pose {
  path=AnimGraph.AnimGraphNode_StateMachine_0.MagBulletRotation_StateMachine.AnimStateNode_9.Empty pose
  nodes [Output Animation Pose(Result=Sequence Player 'ChargeSMG_BulletsEmpty_Pose_Charged_Rifle'() {sequence=/DaisyWeaponGameplay/Items/Weapons/ChargeSMG/Animation/MagazineRotation/ChargeSMG_BulletsEmpty_Pose_Charged_Rifle})]
 }
 Transition {
  path=AnimGraph.AnimGraphNode_StateMachine_0.MagBulletRotation_StateMachine.AnimStateTransitionNode_100.Transition
  nodes [
   n1 = Get ClipRemaining() -> (ClipRemaining:int=0)
   Result(bCanEnterTransition=AND Boolean(A=Get Recently Fired() -> (Recently Fired:bool), B="integer < integer"(A=@n1, B=18)))
  ]
 }
 Transition {
  path=AnimGraph.AnimGraphNode_StateMachine_0.MagBulletRotation_StateMachine.AnimStateTransitionNode_101.Transition
  nodes [
   n1 = Get ClipRemaining() -> (ClipRemaining=0)
   Result(bCanEnterTransition=AND Boolean(A=Get Recently Fired(), B="integer < integer"(A=@n1, B=9)))
  ]
 }
 Transition {
  path=AnimGraph.AnimGraphNode_StateMachine_0.MagBulletRotation_StateMachine.AnimStateTransitionNode_12.Transition
  nodes [
   n1 = Get ClipRemaining() -> (ClipRemaining=0)
   Result(bCanEnterTransition=AND Boolean(A=Get Recently Fired(), B="integer < integer"(A=@n1, B=3)))
  ]
 }
 Transition {
  path=AnimGraph.AnimGraphNode_StateMachine_0.MagBulletRotation_StateMachine.AnimStateTransitionNode_13.Transition
  nodes [Result(bCanEnterTransition=Get Reloading() -> (Reloading:bool))]
 }
 Transition {
  path=AnimGraph.AnimGraphNode_StateMachine_0.MagBulletRotation_StateMachine.AnimStateTransitionNode_14.Transition
  nodes [
   n1 = Get RemainingAmmo() -> (RemainingAmmo:int=0)
   Result(bCanEnterTransition=AND Boolean(A=Get ReloadDone() -> (ReloadDone:bool), B="integer > integer"(A=@n1, B=27)))
   Get ClipRemaining() -> (ClipRemaining=0)
  ]
 }
 Transition {
  path=AnimGraph.AnimGraphNode_StateMachine_0.MagBulletRotation_StateMachine.AnimStateTransitionNode_24.Transition
  nodes [Result(bCanEnterTransition=Get Reloading())]
 }
 Transition {
  path=AnimGraph.AnimGraphNode_StateMachine_0.MagBulletRotation_StateMachine.AnimStateTransitionNode_26.Transition
  nodes [Result(bCanEnterTransition=Get Reloading())]
 }
 Transition {
  path=AnimGraph.AnimGraphNode_StateMachine_0.MagBulletRotation_StateMachine.AnimStateTransitionNode_27.Transition
  nodes [Result(bCanEnterTransition=Get Reloading())]
 }
 Transition {
  path=AnimGraph.AnimGraphNode_StateMachine_0.MagBulletRotation_StateMachine.AnimStateTransitionNode_29.Transition
  nodes [
   n1 = Get RemainingAmmo() -> (RemainingAmmo=0)
   Result(bCanEnterTransition=AND Boolean(A=Get ReloadDone(), B="integer < integer"(A=@n1, B=27)))
   Get ClipRemaining() -> (ClipRemaining=0)
  ]
 }
 Transition {
  path=AnimGraph.AnimGraphNode_StateMachine_0.MagBulletRotation_StateMachine.AnimStateTransitionNode_30.Transition
  nodes [Result(bCanEnterTransition=Get Reloading())]
 }
 Transition {
  path=AnimGraph.AnimGraphNode_StateMachine_0.MagBulletRotation_StateMachine.AnimStateTransitionNode_31.Transition
  nodes [Result(bCanEnterTransition=Get Reloading())]
 }
 Transition {
  path=AnimGraph.AnimGraphNode_StateMachine_0.MagBulletRotation_StateMachine.AnimStateTransitionNode_32.Transition
  nodes [Result(bCanEnterTransition=Get Reloading())]
 }
 Transition {
  path=AnimGraph.AnimGraphNode_StateMachine_0.MagBulletRotation_StateMachine.AnimStateTransitionNode_33.Transition
  nodes [Result(bCanEnterTransition=Get Reloading())]
 }
 Transition {
  path=AnimGraph.AnimGraphNode_StateMachine_0.MagBulletRotation_StateMachine.AnimStateTransitionNode_34.Transition
  nodes [Result(bCanEnterTransition=Get Reloading())]
 }
 Transition {
  path=AnimGraph.AnimGraphNode_StateMachine_0.MagBulletRotation_StateMachine.AnimStateTransitionNode_35.Transition
  nodes [Result(bCanEnterTransition=Get Reloading())]
 }
 Transition {
  path=AnimGraph.AnimGraphNode_StateMachine_0.MagBulletRotation_StateMachine.AnimStateTransitionNode_36.Transition
  nodes [
   n1 = Get ClipRemaining() -> (ClipRemaining=0)
   Result(bCanEnterTransition=AND Boolean(A="integer < integer"(A=@n1, B=28), B="integer > integer"(A=@n1, B=24)))
   Get RemainingAmmo() -> (RemainingAmmo=0)
  ]
 }
 Transition {
  path=AnimGraph.AnimGraphNode_StateMachine_0.MagBulletRotation_StateMachine.AnimStateTransitionNode_37.Transition
  nodes [
   n1 = Get ClipRemaining() -> (ClipRemaining=0)
   Result(bCanEnterTransition=AND Boolean(A="integer < integer"(A=@n1, B=25), B="integer > integer"(A=@n1, B=21)))
   Get RemainingAmmo() -> (RemainingAmmo=0)
  ]
 }
 Transition {
  path=AnimGraph.AnimGraphNode_StateMachine_0.MagBulletRotation_StateMachine.AnimStateTransitionNode_38.Transition
  nodes [
   n1 = Get ClipRemaining() -> (ClipRemaining=0)
   Result(bCanEnterTransition=AND Boolean(A="integer < integer"(A=@n1, B=22), B="integer > integer"(A=@n1, B=18)))
   Get RemainingAmmo() -> (RemainingAmmo=0)
  ]
 }
 Transition {
  path=AnimGraph.AnimGraphNode_StateMachine_0.MagBulletRotation_StateMachine.AnimStateTransitionNode_39.Transition
  nodes [
   n1 = Get ClipRemaining() -> (ClipRemaining=0)
   Result(bCanEnterTransition=AND Boolean(A="integer < integer"(A=@n1, B=19), B="integer > integer"(A=@n1, B=15)))
  ]
 }
 Transition {
  path=AnimGraph.AnimGraphNode_StateMachine_0.MagBulletRotation_StateMachine.AnimStateTransitionNode_40.Transition
  nodes [
   n1 = Get ClipRemaining() -> (ClipRemaining=0)
   Result(bCanEnterTransition=AND Boolean(A="integer < integer"(A=@n1, B=16), B="integer > integer"(A=@n1, B=12)))
  ]
 }
 Transition {
  path=AnimGraph.AnimGraphNode_StateMachine_0.MagBulletRotation_StateMachine.AnimStateTransitionNode_41.Transition
  nodes [
   n1 = Get ClipRemaining() -> (ClipRemaining=0)
   Result(bCanEnterTransition=AND Boolean(A="integer < integer"(A=@n1, B=13), B="integer > integer"(A=@n1, B=9)))
  ]
 }
 Transition {
  path=AnimGraph.AnimGraphNode_StateMachine_0.MagBulletRotation_StateMachine.AnimStateTransitionNode_42.Transition
  nodes [
   n1 = Get ClipRemaining() -> (ClipRemaining=0)
   Result(bCanEnterTransition=AND Boolean(A="integer < integer"(A=@n1, B=10), B="integer > integer"(A=@n1, B=6)))
  ]
 }
 Transition {
  path=AnimGraph.AnimGraphNode_StateMachine_0.MagBulletRotation_StateMachine.AnimStateTransitionNode_43.Transition
  nodes [
   n1 = Get ClipRemaining() -> (ClipRemaining=0)
   Result(bCanEnterTransition=AND Boolean(A="integer < integer"(A=@n1, B=7), B="integer > integer"(A=@n1, B=3)))
  ]
 }
 Transition {
  path=AnimGraph.AnimGraphNode_StateMachine_0.MagBulletRotation_StateMachine.AnimStateTransitionNode_44.Transition
  nodes [
   n1 = Get ClipRemaining() -> (ClipRemaining=0)
   Result(bCanEnterTransition=AND Boolean(A="integer < integer"(A=@n1, B=4), B="integer > integer"(A=@n1)))
  ]
 }
 Transition {
  path=AnimGraph.AnimGraphNode_StateMachine_0.MagBulletRotation_StateMachine.AnimStateTransitionNode_6.Transition
  nodes [
   n1 = Get ClipRemaining() -> (ClipRemaining=0)
   Result(bCanEnterTransition=AND Boolean(A=Get Recently Fired(), B="integer < integer"(A=@n1, B=15)))
  ]
 }
 Transition {
  path=AnimGraph.AnimGraphNode_StateMachine_0.MagBulletRotation_StateMachine.AnimStateTransitionNode_7.Transition
  nodes [
   n1 = Get ClipRemaining() -> (ClipRemaining=0)
   Result(bCanEnterTransition=AND Boolean(A=Get Recently Fired(), B="integer < integer"(A=@n1, B=12)))
  ]
 }
 Transition {
  path=AnimGraph.AnimGraphNode_StateMachine_0.MagBulletRotation_StateMachine.AnimStateTransitionNode_9.Transition
  nodes [
   n1 = Get ClipRemaining() -> (ClipRemaining=0)
   Result(bCanEnterTransition=AND Boolean(A=Get Recently Fired(), B="integer < integer"(A=@n1, B=6)))
  ]
 }
 Transition {
  path=AnimGraph.AnimGraphNode_StateMachine_0.MagBulletRotation_StateMachine.AnimStateTransitionNode_96.Transition
  nodes [
   n1 = Get ClipRemaining() -> (ClipRemaining=0)
   Result(bCanEnterTransition=AND Boolean(A=Get Recently Fired(), B="integer < integer"(A=@n1, B=30)))
   "integer < integer"(A=Get RemainingAmmo() -> (RemainingAmmo=0), B=30)
  ]
 }
 Transition {
  path=AnimGraph.AnimGraphNode_StateMachine_0.MagBulletRotation_StateMachine.AnimStateTransitionNode_97.Transition
  nodes [
   n1 = Get ClipRemaining() -> (ClipRemaining=0)
   Result(bCanEnterTransition=AND Boolean(A=Get Recently Fired(), B="integer < integer"(A=@n1, B=27)))
  ]
 }
 Transition {
  path=AnimGraph.AnimGraphNode_StateMachine_0.MagBulletRotation_StateMachine.AnimStateTransitionNode_98.Transition
  nodes [
   n1 = Get ClipRemaining() -> (ClipRemaining=0)
   Result(bCanEnterTransition=AND Boolean(A=Get Recently Fired(), B="integer < integer"(A=@n1, B=24)))
  ]
 }
 Transition {
  path=AnimGraph.AnimGraphNode_StateMachine_0.MagBulletRotation_StateMachine.AnimStateTransitionNode_99.Transition
  nodes [
   n1 = Get ClipRemaining() -> (ClipRemaining=0)
   Result(bCanEnterTransition=AND Boolean(A=Get Recently Fired(), B="integer < integer"(A=@n1, B=21)))
  ]
 }
 BulletInsert_StateMachine {
  entry=Pre-Bullet Insert
  path=AnimGraph.AnimGraphNode_StateMachine_1.BulletInsert_StateMachine
  ids=id
  nodes [
   AnimStateNode_0 = Bullets Inserted() {is=title AnimStateNode}
   AnimStateNode_2 = Empty Mag() {is=title AnimStateNode}
   AnimStateNode_3 = Pre-Bullet Insert() {is=title AnimStateNode}
   AnimStateNode_4 = LastShot() {is=title AnimStateNode}
   AnimStateNode_5 = Pre-LastShot() {is=title AnimStateNode}
   AnimStateTransitionNode_0 = AnimStateTransitionNode() {bAutomaticRuleBasedOnSequencePlayerInState=true, crossfadeDuration=0, priorityOrder=2}
   AnimStateTransitionNode_11 = AnimStateTransitionNode() {bAutomaticRuleBasedOnSequencePlayerInState=true, crossfadeDuration=0.1}
   AnimStateTransitionNode_13 = AnimStateTransitionNode() {crossfadeDuration=0}
   AnimStateTransitionNode_14 = AnimStateTransitionNode()
   AnimStateTransitionNode_2 = AnimStateTransitionNode()
   AnimStateTransitionNode_5 = AnimStateTransitionNode() {crossfadeDuration=0}
  ]
  links [
   {
    dst=Pre-Bullet Insert
    rule=AnimGraph.AnimGraphNode_StateMachine_1.BulletInsert_StateMachine.AnimStateTransitionNode_0.Transition
    src=Bullets Inserted
    via=AnimStateTransitionNode_0
   }
   {
    dst=Pre-LastShot
    rule=AnimGraph.AnimGraphNode_StateMachine_1.BulletInsert_StateMachine.AnimStateTransitionNode_13.Transition
    src=Bullets Inserted
    via=AnimStateTransitionNode_13
   }
   {
    dst=Pre-Bullet Insert
    rule=AnimGraph.AnimGraphNode_StateMachine_1.BulletInsert_StateMachine.AnimStateTransitionNode_2.Transition
    src=Empty Mag
    via=AnimStateTransitionNode_2
   }
   {
    dst=Empty Mag
    rule=AnimGraph.AnimGraphNode_StateMachine_1.BulletInsert_StateMachine.AnimStateTransitionNode_11.Transition
    src=LastShot
    via=AnimStateTransitionNode_11
   }
   {
    dst=Bullets Inserted
    rule=AnimGraph.AnimGraphNode_StateMachine_1.BulletInsert_StateMachine.AnimStateTransitionNode_5.Transition
    src=Pre-Bullet Insert
    via=AnimStateTransitionNode_5
   }
   {
    dst=LastShot
    rule=AnimGraph.AnimGraphNode_StateMachine_1.BulletInsert_StateMachine.AnimStateTransitionNode_14.Transition
    src=Pre-LastShot
    via=AnimStateTransitionNode_14
   }
  ]
 }
 Bullets Inserted {
  path=AnimGraph.AnimGraphNode_StateMachine_1.BulletInsert_StateMachine.AnimStateNode_0.Bullets Inserted
  nodes [
   n1 = Sequence Player 'ChargeSMG_BulletInsert_Loop_Charged_Rifle'() {
    bLoopAnimation=false
    sequence=/DaisyWeaponGameplay/Items/Weapons/ChargeSMG/Animation/MagazineRotation/ChargeSMG_BulletInsert_Loop_Charged_Rifle
   }
   Output Animation Pose(Result=@n1)
  ]
 }
 Empty Mag {
  path=AnimGraph.AnimGraphNode_StateMachine_1.BulletInsert_StateMachine.AnimStateNode_2.Empty Mag
  nodes [Output Animation Pose(Result=Sequence Player 'ChargeSMG_BulletsEmpty_Pose_Charged_Rifle'() {sequence=&ChargeSMG_BulletsEmpty_Pose_Charged_Rifle})]
 }
 Pre-Bullet Insert {
  path=AnimGraph.AnimGraphNode_StateMachine_1.BulletInsert_StateMachine.AnimStateNode_3.Pre-Bullet Insert
  nodes [Output Animation Pose(Result=Sequence Player 'ChargeSMG_Pre_BulletInsert_Pose_Charged_Rifle'() {sequence=/DaisyWeaponGameplay/Items/Weapons/ChargeSMG/Animation/MagazineRotation/ChargeSMG_Pre_BulletInsert_Pose_Charged_Rifle})]
 }
 LastShot {
  path=AnimGraph.AnimGraphNode_StateMachine_1.BulletInsert_StateMachine.AnimStateNode_4.LastShot
  nodes [Output Animation Pose(Result=Sequence Player 'ChargeSMG_BulletInsert_Loop_Charged_Rifle'() {sequence=&ChargeSMG_BulletInsert_Loop_Charged_Rifle})]
 }
 Pre-LastShot {
  path=AnimGraph.AnimGraphNode_StateMachine_1.BulletInsert_StateMachine.AnimStateNode_5.Pre-LastShot
  nodes [Output Animation Pose(Result=Sequence Player 'ChargeSMG_Pre_BulletInsert_Pose_Charged_Rifle'() {sequence=&ChargeSMG_Pre_BulletInsert_Pose_Charged_Rifle})]
 }
 Transition {path=AnimGraph.AnimGraphNode_StateMachine_1.BulletInsert_StateMachine.AnimStateTransitionNode_0.Transition, nodes [Result()]}
 Transition {path=AnimGraph.AnimGraphNode_StateMachine_1.BulletInsert_StateMachine.AnimStateTransitionNode_11.Transition, nodes [Result()]}
 Transition {
  path=AnimGraph.AnimGraphNode_StateMachine_1.BulletInsert_StateMachine.AnimStateTransitionNode_13.Transition
  nodes [Result(bCanEnterTransition=Get LastShot() -> (LastShot:bool))]
 }
 Transition {
  path=AnimGraph.AnimGraphNode_StateMachine_1.BulletInsert_StateMachine.AnimStateTransitionNode_14.Transition
  nodes [
   n1 = Get ClipRemaining() -> (ClipRemaining=0)
   Result(bCanEnterTransition=AND Boolean(A=Get Recently Fired(), B="integer < integer"(A=@n1, B=3)))
  ]
 }
 Transition {
  path=AnimGraph.AnimGraphNode_StateMachine_1.BulletInsert_StateMachine.AnimStateTransitionNode_2.Transition
  nodes [Result(bCanEnterTransition=Get Reloading())]
 }
 Transition {
  path=AnimGraph.AnimGraphNode_StateMachine_1.BulletInsert_StateMachine.AnimStateTransitionNode_5.Transition
  nodes [
   n1 = Get ClipRemaining() -> (ClipRemaining=0)
   Result(bCanEnterTransition=AND Boolean(A=Get Recently Fired(), B="integer > integer"(A=@n1, B=3)))
  ]
 }
 EventGraph {nodes [
   Event BlueprintUpdateAnimation()
   n1 = GetOwningActor()
   > Is Valid(InputObject=@n1)
   > n2 = Cast To B_ChargeSMG_Athena!(execute, Object=@n1) -> (then, CastFailed, AsB Charge SMG Athena:B Charge SMG Athena Object Reference) {is=title K2Node_DynamicCast}
   > n3 = Set WeapBP(WeapBP:B Charge SMG Athena Object Reference=@n2) -> (Output_Get:B Charge SMG Athena Object Reference)
   > Set ChargePercent(ChargePercent:double=GetChargePercent!(self:Fort Weapon Object Reference=@n3) -> (ReturnValue:float)) -> (Output_Get:double)
   > n4 = Sequence()
   n5 = Get ChargePercent() -> (ChargePercent:double)
   > n6 = Branch(Condition="float > float"(A=@n5))
   n7 = Get ChargePercent()
   > n8 = Set ChargeRotation(ChargeRotation:double=float * float(A="Clamp (Float)"(Value=@n7, Min=0, Max=1), B=270) {is=title K2Node_PromotableOperator}) -> (Output_Get:double)
   > Set CachedChargeRotation(CachedChargeRotation:double=@n8) -> (Output_Get:double)
   > Set InterpTimeTotal(InterpTimeTotal:double=float * float(A=@n5, B=1.8)) -> (Output_Get:double)
   > Set InterpTime(InterpTime:double=0) -> (Output_Get:double)
   n9 = Set InterpTime(execute=@n6.else, InterpTime=float + float(A=GetWorldDeltaSeconds!() -> (ReturnValue:double), B=Get InterpTime() -> (InterpTime:double)))
   n10 = SafeDivide(A=@n9, B=Get InterpTimeTotal() -> (InterpTimeTotal:double))
   > Set ChargeRotation(ChargeRotation=Lerp(A=Get CachedChargeRotation() -> (CachedChargeRotation:double), B=0, Alpha="Clamp (Float)"(Value=@n10, Min=0, Max=1)))
   n11 = Get WeapBP() -> (WeapBP:B Charge SMG Athena Object Reference)
   Set Recently Fired(execute=@n4.then_1, Recently Fired:bool="float > float"(A=GetTimeToNextFire!(self:Fort Weapon Object Reference=@n11) -> (ReturnValue:float))) -> (Output_Get:bool)
   n12 = GetMagazineAmmoCount!(self:Fort Weapon Object Reference=@n11) -> (ReturnValue:int=0)
   > Set BulletsUsed(BulletsUsed:int=int - int(A=30, B=@n12)) -> (Output_Get:int=0)
   > Set LastShot(LastShot:bool="integer <= integer"(A=@n12, B=3)) -> (Output_Get:bool)
   > Set ClipRemaining(ClipRemaining:int=@n12) -> (Output_Get:int=0)
   > Set Reloading(Reloading:bool=IsReloading!(self:Fort Weapon Object Reference=@n11) -> (ReturnValue:bool) {is=title K2Node_CallFunction}) -> (Output_Get:bool)
   > Set ReloadDone(ReloadDone:bool="Equal (Float)"(A=GetCurveValue(CurveName=Reloaded), B=1)) -> (Output_Get:bool)
   > Set RemainingAmmo(RemainingAmmo:int=GetLocalRemainingAmmo!(self:Fort Weapon Object Reference=@n11) -> (ReturnValue:int=0)) -> (Output_Get:int=0)
   n13 = Get WeapBP()
   n14 = "To Float (Integer)"(InInt=GetBulletsPerClip!(self:Fort Weapon Object Reference=@n13) -> (ReturnValue:int=0))
   n15 = "To Float (Integer)"(InInt=GetMagazineAmmoCount(self=@n13) -> (ReturnValue=0))
   float + float(A=@n7, B=float - float(A=1, B=float / float(A:double=@n15, B:double=@n14)))
   float / float(A:Wildcard, B:Wildcard) -> (ReturnValue:Wildcard)
  ]}
]
variables [
 BulletsUsed {type=integer}
 CachedChargeRotation {type=number}
 ChargePercent {type=number}
 ChargeRotation {type=number}
 ClipRemaining {type=integer, default=30}
 InterpTime {type=number}
 InterpTimeTotal {type=number}
 LastShot {type=boolean}
 Recently Fired {type=boolean}
 ReloadDone {type=boolean}
 Reloading {type=boolean}
 RemainingAmmo {type=integer}
 WeapBP {type=B_ChargeSMG_Athena_C, default=null}
]
