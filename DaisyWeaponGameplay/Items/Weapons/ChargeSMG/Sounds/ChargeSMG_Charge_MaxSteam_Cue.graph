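# Sound cue for the Charge SMG's max-steam charge sound, reusing a Charge Shotgun steam wave and charging attenuation.
# A single wave player feeds a slight upward random pitch modulator, then attenuation, then a third-person ranged-weapon-fire sound class override.
# Concurrency is limited per owner, stopping the oldest instance.
# The cue's volume multiplier is zero, so it is silent as saved, and its debug flag is on.
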SoundCue /DaisyWeaponGameplay/Items/Weapons/ChargeSMG/Sounds/ChargeSMG_Charge_MaxSteam_Cue
root=SoundClass_2
nodes [
 {
  id=Attenuation_2
  children [Modulator_2]
  properties {attenuationSettings=/FNGameplayCues/Sounds/Attenuation/Fort_Charge_Shotgun_Charging_Att}
 }
 {id=Modulator_2, children [WavePlayer_1], properties {pitchMax=1.1, pitchMin=1, volumeMax=1, volumeMin=1}}
 {
  id=SoundClass_2
  children [Attenuation_2]
  properties {soundClassOverride=&Fort_WeaponRanged_Fire_3P_PSM}
 }
 {id=WavePlayer_1, properties {soundWaveAssetPtr=/FNGameplayCues/Sounds/Fort_Weapons/Shotgun/Charge/SG_Charge_Startup_Steam_Max_01_v2}}
]
properties {
 bDebug=true
 bOverrideConcurrency=true
 maxDistance=3000
 soundClassObject=&Fort_WeaponRanged_Fire_3P_PSM
 volumeMultiplier=0
 concurrencyOverrides {bLimitToOwner=true, maxCount=3, resolutionRule=StopOldest}
}
