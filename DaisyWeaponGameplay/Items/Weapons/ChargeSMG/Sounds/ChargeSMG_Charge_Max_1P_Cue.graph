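# First-person sound cue for the Charge SMG's maximum charge, using the first-person ranged-weapon fire sound class.
# A Mixer layers two charge-full waves and a charge-start layer at different volumes, with one charge-full wave pitched down by a fixed Modulator.
# The mix passes through first-person Athena projectile attenuation and a sound class override node.
# Overridden concurrency caps instances per owner and stops the oldest.
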SoundCue /DaisyWeaponGameplay/Items/Weapons/ChargeSMG/Sounds/ChargeSMG_Charge_Max_1P_Cue
root=SoundClass_2
nodes [
 {
  id=Attenuation_2
  children [Mixer_0]
  properties {attenuationSettings=&Fort_Weapon_Projectile_FirstPerson_Athena_Att}
 }
 {id=Mixer_0, children [WavePlayer_5, WavePlayer_10, Modulator_0], properties {inputVolume [0.8,0.4,0.6]}}
 {id=Modulator_0, children [WavePlayer_0], properties {pitchMax=0.9, pitchMin=0.9}}
 {
  id=SoundClass_2
  children [Attenuation_2]
  properties {soundClassOverride=&Fort_WeaponRanged_Fire_1P_PSM}
 }
 {id=WavePlayer_0, properties {soundWaveAssetPtr=&/ChargeSMG_ChargeFull_03}}
 {id=WavePlayer_10, properties {soundWaveAssetPtr=&/ChargeSMG_ChargeStartLayer_05}}
 {id=WavePlayer_5, properties {soundWaveAssetPtr=&/ChargeSMG_ChargeFull_01}}
]
properties {
 bDebug=true
 bOverrideConcurrency=true
 maxDistance=1800.0001
 soundClassObject=&Fort_WeaponRanged_Fire_1P_PSM
 volumeMultiplier=0.4
 concurrencyOverrides {bLimitToOwner=true, maxCount=3, resolutionRule=StopOldest}
}
