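# Sound cue for the first-person charge-start sound of the ChargeSMG weapon.
# A single wave feeds a one-entry Random node and a default Modulator, then passes through attenuation and sound class nodes.
# It uses first-person projectile attenuation and the first-person ranged-fire sound class, at reduced volume.
# A concurrency override caps instances per owner and stops the oldest, and the debug flag is on.
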SoundCue /DaisyWeaponGameplay/Items/Weapons/ChargeSMG/Sounds/ChargeSMG_Charge_Start_1P_Cue
root=SoundClass_2
nodes [
 {
  id=Attenuation_2
  children [Modulator_0]
  properties {attenuationSettings=&Fort_Weapon_Projectile_FirstPerson_Athena_Att}
 }
 {id=Modulator_0, children [Random_0]}
 {id=Random_0, children [WavePlayer_7], properties {weights [1]}}
 {
  id=SoundClass_2
  children [Attenuation_2]
  properties {soundClassOverride=&Fort_WeaponRanged_Fire_1P_PSM}
 }
 {id=WavePlayer_7, properties {soundWaveAssetPtr=&/ChargeSMG_ChargeStart_01}}
]
properties {
 attenuationSettings=&Fort_Weapon_Projectile_FirstPerson_Athena_Att
 bDebug=true
 bOverrideConcurrency=true
 maxDistance=2097152
 soundClassObject=&Fort_WeaponRanged_Fire_1P_PSM
 volumeMultiplier=0.25
 concurrencyOverrides {bLimitToOwner=true, maxCount=3, resolutionRule=StopOldest}
}
