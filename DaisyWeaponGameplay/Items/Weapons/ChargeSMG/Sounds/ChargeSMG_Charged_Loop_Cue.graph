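# Sound cue for the looping sound of the Charge SMG weapon in its charged state.
# A mixer blends three looping waves, a main charge loop and two layers, at individual volumes, then feeds a charging attenuation node.
# A sound class node at the root applies the third-person ranged-fire class.
# The cue lowers overall volume, raises pitch, and overrides concurrency to a few per-owner instances, stopping the oldest.
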SoundCue /DaisyWeaponGameplay/Items/Weapons/ChargeSMG/Sounds/ChargeSMG_Charged_Loop_Cue
root=SoundClass_2
nodes [
 {id=Attenuation_2, children [Mixer_0], properties {attenuationSettings=/Game/Sounds/Attenuation/Fort_Charge_SMG_Charging_Att}}
 {id=Mixer_0, children [WavePlayer_2,WavePlayer_4,WavePlayer_7], properties {inputVolume [0.85,1.1,1.1]}}
 {
  id=SoundClass_2
  children [Attenuation_2]
  properties {soundClassOverride=&Fort_WeaponRanged_Fire_3P_PSM}
 }
 {id=WavePlayer_2, properties {bLooping=true, soundWaveAssetPtr=&/ChargeSMG_ChargeLoop_03}}
 {
  id=WavePlayer_4
  properties {bLooping=true, soundWaveAssetPtr=&/ChargeSMG_ChargeLoopLayer_01}
 }
 {
  id=WavePlayer_7
  properties {bLooping=true, soundWaveAssetPtr=&/ChargeSMG_ChargeLoopLayer_04}
 }
]
properties {
 bDebug=true
 bOverrideConcurrency=true
 maxDistance=3700
 pitchMultiplier=1.2
 soundClassObject=&Fort_WeaponRanged_Fire_3P_PSM
 volumeMultiplier=0.22
 concurrencyOverrides {bLimitToOwner=true, maxCount=3, resolutionRule=StopOldest}
}
