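# First-person fire sound cue for the Charge SMG, limited per owner with the oldest instance stopped first.
# Random close-shot and laser layers mix with quieter, occasionally silent casings, and a TimeSinceStarted parameter modulates that mix's volume downward.
# A LowAmmo branch adds a boosted, pitched-up warning under its own critical low-ammo sound class.
# The FirstShot branch has no sounds connected.
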SoundCue /DaisyWeaponGameplay/Items/Weapons/ChargeSMG/Sounds/ChargeSMG_Fire_1P_Cue
root=Mixer_6
nodes [
 {id=Branch_0, children [null,null,null], properties {boolParameterName=FirstShot}}
 {id=Branch_5, children [Modulator_12,null,null], properties {boolParameterName=LowAmmo}}
 {id=Delay_5, children [null]}
 {id=Mixer_0, children [Random_17,Random_0], properties {inputVolume [1,0.7]}}
 {id=Mixer_5, children [Modulator_8,Modulator_9,Delay_5], properties {inputVolume [1,0.4,0]}}
 {id=Mixer_6, children [SoundClass_2,SoundClass_3], properties {inputVolume [1.2,1]}}
 {id=Mixer_7, children [Branch_0, ModulatorContinuous_1], properties {inputVolume [1,1]}}
 {
  id=ModulatorContinuous_1
  children [Modulator_10]
  properties {volumeModulationParams {default=0, maxOutput=0.6, minOutput=1, parameterName=TimeSinceStarted}}
 }
 {id=Modulator_10, children [Mixer_5], properties {pitchMax=1, pitchMin=1}}
 {id=Modulator_11, children [Branch_5], properties {pitchMax=1, pitchMin=1, volumeMax=1.15, volumeMin=1}}
 {id=Modulator_12, children [WavePlayer_90], properties {pitchMax=1.1, pitchMin=1.1, volumeMax=1.8, volumeMin=1.8}}
 {id=Modulator_8, children [Mixer_0], properties {pitchMax=1, pitchMin=1}}
 {id=Modulator_9, children [Random_18], properties {volumeMax=1.25}}
 {id=Random_0, children [WavePlayer_2,WavePlayer_5,WavePlayer_6], properties {weights [1,1,1]}}
 {
  id=Random_17
  children [WavePlayer_91, WavePlayer_92, WavePlayer_93, WavePlayer_94, WavePlayer_95, WavePlayer_96]
  properties {weights [1,1,1,1,1,1]}
 }
 {
  id=Random_18
  children [WavePlayer_97, WavePlayer_98, WavePlayer_99, WavePlayer_100, WavePlayer_101, null, null, null, null]
  properties {weights [1,1,1,1,1,1,1,1,1]}
 }
 {
  id=SoundClass_2
  children [Modulator_11]
  properties {soundClassOverride=/Game/Sounds/FortSoundClassesAndModes/Fort_WeaponRanged_LowAmmo_Critical_PSM}
 }
 {id=SoundClass_3, children [Mixer_7], properties {soundClassOverride=&Fort_WeaponRanged_Fire_1P_PSM}}
 {id=WavePlayer_100, properties {soundWaveAssetPtr=&/ChargeSMG_Casings_04}}
 {id=WavePlayer_101, properties {soundWaveAssetPtr=&/ChargeSMG_Casings_05}}
 {id=WavePlayer_2, properties {soundWaveAssetPtr=&/ChargeSMG_Fire_LaserLayer_01}}
 {id=WavePlayer_5, properties {soundWaveAssetPtr=&/ChargeSMG_Fire_LaserLayer_04}}
 {id=WavePlayer_6, properties {soundWaveAssetPtr=&/ChargeSMG_Fire_LaserLayer_05}}
 {id=WavePlayer_90, properties {soundWaveAssetPtr=&/ChargeSMG_LowAmmo_02}}
 {id=WavePlayer_91, properties {soundWaveAssetPtr=&/ChargeSMG_Fire_Close_01}}
 {id=WavePlayer_92, properties {soundWaveAssetPtr=&/ChargeSMG_Fire_Close_02}}
 {id=WavePlayer_93, properties {soundWaveAssetPtr=&/ChargeSMG_Fire_Close_03}}
 {id=WavePlayer_94, properties {soundWaveAssetPtr=&/ChargeSMG_Fire_Close_04}}
 {id=WavePlayer_95, properties {soundWaveAssetPtr=&/ChargeSMG_Fire_Close_05}}
 {id=WavePlayer_96, properties {soundWaveAssetPtr=&/ChargeSMG_Fire_Close_06}}
 {id=WavePlayer_97, properties {soundWaveAssetPtr=&/ChargeSMG_Casings_01}}
 {id=WavePlayer_98, properties {soundWaveAssetPtr=&/ChargeSMG_Casings_02}}
 {id=WavePlayer_99, properties {soundWaveAssetPtr=&/ChargeSMG_Casings_03}}
]
properties {
 attenuationSettings=&Fort_Weapon_Projectile_FirstPerson_Athena_Att
 bDebug=true
 maxDistance=2097152
 soundClassObject=&Fort_WeaponRanged_Fire_1P_PSM
 volumeMultiplier=1.1
 concurrencyOverrides {bLimitToOwner=true, resolutionRule=StopOldest}
 concurrencySet [&Weapon_Gunfire_FirstPerson_Conc]
}
