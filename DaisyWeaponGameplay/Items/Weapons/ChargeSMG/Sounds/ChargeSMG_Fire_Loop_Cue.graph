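# Looping fire Sound Cue for the ChargeSMG, built from distance-layered close, medium and distant loops.
# Each looping wave runs through its own pitch and volume modulator, attenuation and sound class override into one weighted mixer.
# Close and medium layers use the third-person fire class and the far layer uses the distant class.
# The cue itself carries first-person attenuation, sound class and gunfire concurrency, with debug enabled.
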SoundCue /DaisyWeaponGameplay/Items/Weapons/ChargeSMG/Sounds/ChargeSMG_Fire_Loop_Cue
root=Mixer_17
nodes [
 {
  id=Attenuation_0
  children [Modulator_4]
  properties {attenuationSettings=&Fort_Weapon_Projectile_Medium_Mono_Athena_Att}
 }
 {
  id=Attenuation_34
  children [Modulator_5]
  properties {attenuationSettings=&Fort_Weapon_Projectile_Distant_Mono_Athena_Att}
 }
 {id=Attenuation_49, children [Modulator_3], properties {attenuationSettings=&Fort_Weapon_Projectile_Close_Athena_Att}}
 {id=Mixer_17, children [SoundClass_32, SoundClass_0, SoundClass_35], properties {inputVolume [0.82,0.47,0.3]}}
 {id=Modulator_3, children [WavePlayer_0], properties {pitchMax=0.95, volumeMax=0.85, volumeMin=0.8}}
 {id=Modulator_4, children [WavePlayer_4], properties {pitchMin=1.05, volumeMax=1.85, volumeMin=1.8}}
 {id=Modulator_5, children [WavePlayer_5], properties {pitchMax=0.95, volumeMax=1.6, volumeMin=1.6}}
 {
  id=SoundClass_0
  children [Attenuation_0]
  properties {soundClassOverride=&Fort_WeaponRanged_Fire_3P_PSM}
 }
 {
  id=SoundClass_32
  children [Attenuation_49]
  properties {soundClassOverride=&Fort_WeaponRanged_Fire_3P_PSM}
 }
 {
  id=SoundClass_35
  children [Attenuation_34]
  properties {soundClassOverride=&Fort_WeaponRanged_Fire_Distant}
 }
 {
  id=WavePlayer_0
  properties {bLooping=true, soundWaveAssetPtr=&/ChargeSMG_Fire_Close_Loop_02}
 }
 {
  id=WavePlayer_4
  properties {bLooping=true, soundWaveAssetPtr=&/ChargeSMG_Fire_Medium_Loop_01}
 }
 {
  id=WavePlayer_5
  properties {bLooping=true, soundWaveAssetPtr=&/ChargeSMG_Fire_Distant_Loop_01}
 }
]
properties {
 attenuationSettings=&Fort_Weapon_Projectile_FirstPerson_Athena_Att
 bDebug=true
 maxDistance=2097152
 soundClassObject=&Fort_WeaponRanged_Fire_1P_PSM
 volumeMultiplier=1
 concurrencySet [&Weapon_Gunfire_FirstPerson_Conc]
}
