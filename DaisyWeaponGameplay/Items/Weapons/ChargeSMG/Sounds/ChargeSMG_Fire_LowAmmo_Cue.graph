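# Sound cue for the Charge SMG's low-ammo firing sound in DaisyWeaponGameplay.
# A single wave player plays the shared Fortnite assault rifle low-ammo loop sample at a boosted volume, with no randomization or mixing.
# It routes to the degraded ranged-weapon low-ammo sound class, uses an extremely large max distance, and overrides concurrency to block new instances rather than cut playing ones.
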SoundCue /DaisyWeaponGameplay/Items/Weapons/ChargeSMG/Sounds/ChargeSMG_Fire_LowAmmo_Cue
root=WavePlayer_70
nodes [{id=WavePlayer_70, properties {soundWaveAssetPtr=/Game/Sounds/Fort_Weapons/LowAmmo/ar_lowammo_loop_03}}]
properties {
 bOverrideConcurrency=true
 maxDistance=2097152
 soundClassObject=/Game/Sounds/FortSoundClassesAndModes/Fort_WeaponRanged_LowAmmo_Degraded_PSM
 volumeMultiplier=1.5
 concurrencyOverrides {resolutionRule=PreventNew}
}
