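# First-person sound cue for the Charge SMG's fire stop, using first-person weapon attenuation, gunfire concurrency and a ranged-fire sound class.
# A mixer layers one of six random close gunshot waves, a quieter charge-stop wave, and one of five random shell casing waves that starts after a short fixed delay.
# Modulators vary pitch and volume on each layer so repeated plays differ.
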SoundCue /DaisyWeaponGameplay/Items/Weapons/ChargeSMG/Sounds/ChargeSMG_Fire_Stop_1P_Cue
root=Mixer_0
nodes [
 {id=Delay_0, children [Modulator_9], properties {delayMax=0.1, delayMin=0.1}}
 {id=Mixer_0, children [Modulator_1,Modulator_0,Delay_0], properties {inputVolume [0.7,0.27,1]}}
 {id=Modulator_0, children [WavePlayer_0], properties {pitchMax=1.1, pitchMin=1, volumeMin=0.9}}
 {id=Modulator_1, children [Random_0], properties {pitchMax=1.1, pitchMin=0.9, volumeMax=1.1, volumeMin=0.9}}
 {id=Modulator_9, children [Random_18], properties {volumeMax=1.25}}
 {
  id=Random_0
  children [WavePlayer_1, WavePlayer_2, WavePlayer_7, WavePlayer_8, WavePlayer_9, WavePlayer_10]
  properties {weights [1,1,1,1,1,1]}
 }
 {id=Random_18, children [WavePlayer_97, WavePlayer_98, WavePlayer_99, WavePlayer_100, WavePlayer_101], properties {weights [1,1,1,1,1]}}
 {id=WavePlayer_0, properties {soundWaveAssetPtr=&/ChargeSMG_ChargeStop_01}}
 {id=WavePlayer_1, properties {soundWaveAssetPtr=&/ChargeSMG_Fire_Close_01}}
 {id=WavePlayer_10, properties {soundWaveAssetPtr=&/ChargeSMG_Fire_Close_06}}
 {id=WavePlayer_100, properties {soundWaveAssetPtr=&/ChargeSMG_Casings_04}}
 {id=WavePlayer_101, properties {soundWaveAssetPtr=&/ChargeSMG_Casings_05}}
 {id=WavePlayer_2, properties {soundWaveAssetPtr=&/ChargeSMG_Fire_Close_02}}
 {id=WavePlayer_7, properties {soundWaveAssetPtr=&/ChargeSMG_Fire_Close_03}}
 {id=WavePlayer_8, properties {soundWaveAssetPtr=&/ChargeSMG_Fire_Close_04}}
 {id=WavePlayer_9, properties {soundWaveAssetPtr=&/ChargeSMG_Fire_Close_05}}
 {id=WavePlayer_97, properties {soundWaveAssetPtr=&/ChargeSMG_Casings_01}}
 {id=WavePlayer_98, properties {soundWaveAssetPtr=&/ChargeSMG_Casings_02}}
 {id=WavePlayer_99, properties {soundWaveAssetPtr=&/ChargeSMG_Casings_03}}
]
properties {
 attenuationSettings=&Fort_Weapon_Projectile_FirstPerson_Athena_Att
 bDebug=true
 maxDistance=2097152
 soundClassObject=&Fort_WeaponRanged_Fire_1P_PSM
 volumeMultiplier=0.67
 concurrencySet [&Weapon_Gunfire_FirstPerson_Conc]
}
