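# Sound cue for the ChargeSMG's fire-stop sound, played in the world with owner-limited, stop-oldest concurrency.
# A root mixer blends close, medium and distant layers, each with its own attenuation and sound class override.
# Each layer picks a random ChargeSMG stop variation through a volume modulator.
# Medium and distant layers also mix in random SterlingSMG distant loop-stop tails, reused from that weapon and bypassing the modulators.
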SoundCue /DaisyWeaponGameplay/Items/Weapons/ChargeSMG/Sounds/ChargeSMG_Fire_Stop_Cue
root=Mixer_17
nodes [
 {
  id=Attenuation_0
  children [Mixer_0]
  properties {attenuationSettings=&Fort_Weapon_Projectile_Medium_Mono_Athena_Att}
 }
 {
  id=Attenuation_34
  children [Mixer_2]
  properties {attenuationSettings=&Fort_Weapon_Projectile_Distant_Mono_Athena_Att}
 }
 {id=Attenuation_49, children [Modulator_1], properties {attenuationSettings=&Fort_Weapon_Projectile_Close_Athena_Att}}
 {id=Mixer_0, children [Modulator_3,Random_6], properties {inputVolume [1.2,0.7]}}
 {id=Mixer_17, children [SoundClass_32, SoundClass_0, SoundClass_35], properties {inputVolume [1,0.8,0.6]}}
 {id=Mixer_2, children [Modulator_4,Random_8], properties {inputVolume [1.2,0.7]}}
 {id=Modulator_1, children [Random_4], properties {volumeMax=0.9, volumeMin=0.85}}
 {id=Modulator_3, children [Random_5], properties {volumeMax=2.2, volumeMin=2.1}}
 {id=Modulator_4, children [Random_7], properties {volumeMax=2.05, volumeMin=1.95}}
 {id=Random_4, children [WavePlayer_5, WavePlayer_6, WavePlayer_7, WavePlayer_15], properties {weights [1,1,1,1]}}
 {id=Random_5, children [WavePlayer_16, WavePlayer_17, WavePlayer_18, WavePlayer_19], properties {weights [1,1,1,1]}}
 {id=Random_6, children [WavePlayer_20, WavePlayer_21, WavePlayer_22], properties {weights [1,1,1]}}
 {id=Random_7, children [WavePlayer_23, WavePlayer_24, WavePlayer_25, WavePlayer_26], properties {weights [1,1,1,1]}}
 {id=Random_8, children [WavePlayer_27, WavePlayer_28, WavePlayer_29], properties {weights [1,1,1]}}
 {
  id=SoundClass_0
  children [Attenuation_0]
  properties {soundClassOverride=&Fort_WeaponRanged_Fire_3P_PSM}
 }
 {
  id=SoundClass_32
  children [Attenuation_49]
  properties {soundClassOverride=&Fort_WeaponRanged_Fire_3P_PSM}
 }
 {
  id=SoundClass_35
  children [Attenuation_34]
  properties {soundClassOverride=&Fort_WeaponRanged_Fire_Distant}
 }
 {id=WavePlayer_15, properties {soundWaveAssetPtr=&/ChargeSMG_Fire_Close_Stop_04}}
 {id=WavePlayer_16, properties {soundWaveAssetPtr=&/ChargeSMG_Fire_Medium_Stop_01}}
 {id=WavePlayer_17, properties {soundWaveAssetPtr=&/ChargeSMG_Fire_Medium_Stop_02}}
 {id=WavePlayer_18, properties {soundWaveAssetPtr=&/ChargeSMG_Fire_Medium_Stop_03}}
 {id=WavePlayer_19, properties {soundWaveAssetPtr=&/ChargeSMG_Fire_Medium_Stop_04}}
 {id=WavePlayer_20, properties {soundWaveAssetPtr=/Game/Sounds/Fort_Weapons/AutoRifle/SterlingSMG/SterlingSMG_Fire_Distant_LoopStop_01}}
 {id=WavePlayer_21, properties {soundWaveAssetPtr=/Game/Sounds/Fort_Weapons/AutoRifle/SterlingSMG/SterlingSMG_Fire_Distant_LoopStop_02}}
 {id=WavePlayer_22, properties {soundWaveAssetPtr=/Game/Sounds/Fort_Weapons/AutoRifle/SterlingSMG/SterlingSMG_Fire_Distant_LoopStop_03}}
 {id=WavePlayer_23, properties {soundWaveAssetPtr=&/ChargeSMG_Fire_Distant_Stop_01}}
 {id=WavePlayer_24, properties {soundWaveAssetPtr=&/ChargeSMG_Fire_Distant_Stop_02}}
 {id=WavePlayer_25, properties {soundWaveAssetPtr=&/ChargeSMG_Fire_Distant_Stop_03}}
 {id=WavePlayer_26, properties {soundWaveAssetPtr=&/ChargeSMG_Fire_Distant_Stop_04}}
 {id=WavePlayer_27, properties {soundWaveAssetPtr=&SterlingSMG_Fire_Distant_LoopStop_01}}
 {id=WavePlayer_28, properties {soundWaveAssetPtr=&SterlingSMG_Fire_Distant_LoopStop_02}}
 {id=WavePlayer_29, properties {soundWaveAssetPtr=&SterlingSMG_Fire_Distant_LoopStop_03}}
 {id=WavePlayer_5, properties {soundWaveAssetPtr=&/ChargeSMG_Fire_Close_Stop_01}}
 {id=WavePlayer_6, properties {soundWaveAssetPtr=&/ChargeSMG_Fire_Close_Stop_02}}
 {id=WavePlayer_7, properties {soundWaveAssetPtr=&/ChargeSMG_Fire_Close_Stop_03}}
]
properties {
 bDebug=true
 maxDistance=27500
 soundClassObject=&Fort_WeaponRanged_Fire_3P_PSM
 volumeMultiplier=0.7
 concurrencySet [/Game/Sounds/Concurrency_Rules/OneMaximum_Owner_StopOldest]
}
