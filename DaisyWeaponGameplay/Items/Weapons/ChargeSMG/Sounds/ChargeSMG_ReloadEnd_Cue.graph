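# Sound cue for the end of the Charge SMG reload, played as weapon foley.
# A random node picks one of three equally weighted ReloadEndB waves and feeds a root modulator left at its default settings.
# The cue lowers volume, slightly raises pitch and uses the shared weapon reload attenuation asset.
# Its concurrency override blocks new instances once the limit is reached, rather than stopping playing ones.
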SoundCue /DaisyWeaponGameplay/Items/Weapons/ChargeSMG/Sounds/ChargeSMG_ReloadEnd_Cue
root=Modulator_0
nodes [
 {id=Modulator_0, children [Random_1]}
 {id=Random_1, children [WavePlayer_4,WavePlayer_5,WavePlayer_6], properties {weights [1,1,1]}}
 {id=WavePlayer_4, properties {soundWaveAssetPtr=&/ChargeSMG_ReloadEndB_01}}
 {id=WavePlayer_5, properties {soundWaveAssetPtr=&/ChargeSMG_ReloadEndB_02}}
 {id=WavePlayer_6, properties {soundWaveAssetPtr=&/ChargeSMG_ReloadEndB_03}}
]
properties {
 attenuationSettings=&Fort_Weapon_Reload_Att
 bOverrideConcurrency=true
 maxDistance=2250
 pitchMultiplier=1.1
 soundClassObject=&Fort_WeaponFoley
 volumeMultiplier=0.551
 attenuationOverrides {falloffDistance=4100, attenuationShapeExtents {x=900}}
 concurrencyOverrides {resolutionRule=PreventNew}
}
