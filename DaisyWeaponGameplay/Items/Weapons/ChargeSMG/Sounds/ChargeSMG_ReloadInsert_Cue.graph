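# Sound cue for the Charge SMG's reload insert foley, classed as weapon foley.
# A Random node picks among three equally weighted insert recordings and feeds a Modulator at the root.
# The cue slightly lowers pitch and volume, references the shared weapon reload attenuation asset, and overrides concurrency to prevent new instances once the limit is reached.
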SoundCue /DaisyWeaponGameplay/Items/Weapons/ChargeSMG/Sounds/ChargeSMG_ReloadInsert_Cue
root=Modulator_0
nodes [
 {id=Modulator_0, children [Random_0]}
 {id=Random_0, children [WavePlayer_1,WavePlayer_2,WavePlayer_3], properties {weights [1,1,1]}}
 {id=WavePlayer_1, properties {soundWaveAssetPtr=&/ChargeSMG_ReloadInsert_01}}
 {id=WavePlayer_2, properties {soundWaveAssetPtr=&/ChargeSMG_ReloadInsert_02}}
 {id=WavePlayer_3, properties {soundWaveAssetPtr=&/ChargeSMG_ReloadInsert_03}}
]
properties {
 attenuationSettings=&Fort_Weapon_Reload_Att
 bOverrideConcurrency=true
 maxDistance=2250
 pitchMultiplier=0.9
 soundClassObject=&Fort_WeaponFoley
 volumeMultiplier=0.702
 attenuationOverrides {falloffDistance=4100, attenuationShapeExtents {x=900}}
 concurrencyOverrides {resolutionRule=PreventNew}
}
