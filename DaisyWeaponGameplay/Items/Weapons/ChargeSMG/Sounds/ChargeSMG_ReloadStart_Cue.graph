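# Sound cue for the start of the ChargeSMG reload, pairing a reload sound with cartridge bounces.
# A randomly chosen reload-start wave plays immediately, mixed over a quieter submix of two delayed bounce layers.
# Both layers pick from the same three bounce waves, with the earlier pitched higher and the later softer.
# It uses weapon foley class and reload attenuation, and overrides concurrency to prevent new instances at the limit.
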SoundCue /DaisyWeaponGameplay/Items/Weapons/ChargeSMG/Sounds/ChargeSMG_ReloadStart_Cue
root=Mixer_0
nodes [
 {id=Delay_0, children [Modulator_1], properties {delayMax=0.3, delayMin=0.3}}
 {id=Delay_1, children [Modulator_2], properties {delayMax=0.43, delayMin=0.43}}
 {id=Mixer_0, children [Modulator_0,Mixer_1], properties {inputVolume [1,0.5]}}
 {id=Mixer_1, children [Delay_0,Delay_1], properties {inputVolume [0.7,0.6]}}
 {id=Modulator_0, children [Random_0]}
 {id=Modulator_1, children [Random_1], properties {pitchMax=1.5, pitchMin=1.3}}
 {id=Modulator_2, children [Random_2], properties {pitchMax=1.2, pitchMin=1.1, volumeMax=0.8, volumeMin=0.7}}
 {id=Random_0, children [WavePlayer_0,WavePlayer_2,WavePlayer_3], properties {weights [1,1,1]}}
 {id=Random_1, children [WavePlayer_6,WavePlayer_7,WavePlayer_8], properties {weights [1,1,1]}}
 {id=Random_2, children [WavePlayer_9, WavePlayer_10, WavePlayer_11], properties {weights [1,1,1]}}
 {id=WavePlayer_0, properties {soundWaveAssetPtr=&/ChargeSMG_ReloadStart_01}}
 {id=WavePlayer_10, properties {soundWaveAssetPtr=&/ChargeSMG_Cartridge_Bounce_02}}
 {id=WavePlayer_11, properties {soundWaveAssetPtr=&/ChargeSMG_Cartridge_Bounce_03}}
 {id=WavePlayer_2, properties {soundWaveAssetPtr=&/ChargeSMG_ReloadStart_02}}
 {id=WavePlayer_3, properties {soundWaveAssetPtr=&/ChargeSMG_ReloadStart_03}}
 {id=WavePlayer_6, properties {soundWaveAssetPtr=&/ChargeSMG_Cartridge_Bounce_01}}
 {id=WavePlayer_7, properties {soundWaveAssetPtr=&ChargeSMG_Cartridge_Bounce_02}}
 {id=WavePlayer_8, properties {soundWaveAssetPtr=&ChargeSMG_Cartridge_Bounce_03}}
 {id=WavePlayer_9, properties {soundWaveAssetPtr=&ChargeSMG_Cartridge_Bounce_01}}
]
properties {
 attenuationSettings=&Fort_Weapon_Reload_Att
 bOverrideConcurrency=true
 maxDistance=2250
 soundClassObject=&Fort_WeaponFoley
 volumeMultiplier=0.501
 attenuationOverrides {falloffDistance=4100, attenuationShapeExtents {x=900}}
 concurrencyOverrides {resolutionRule=PreventNew}
}
